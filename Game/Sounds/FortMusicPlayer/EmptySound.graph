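# EmptySound is a placeholder SoundCue in the FortMusicPlayer folder that plays silence.
# A single looping wave player node is the root and plays a one-second silence wave, so the cue keeps running without audible output.
# Its max distance is set to a very large value.
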SoundCue /Game/Sounds/FortMusicPlayer/EmptySound
root=WavePlayer_0
nodes [{id=WavePlayer_0, properties {bLooping=true, soundWaveAssetPtr=&/1second_silence}}]
properties {maxDistance=2097152}
